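# Physics asset for the F_MED_Hostile backpack cosmetic, defining the bodies and joints for its dynamic gun, pouch and top parts.
# Each dyn_ bone gets a single capsule body constrained directly to the capsule on the pack bone, forming a star layout.
# The gun and both pouches have swing 2 and twist locked, leaving one swing axis to move.
# The top locks only twist, allowing tighter swing on two axes.
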PhysicsAsset /BRCosmetics/Accessories/FORT_Backpacks/Backpack_F_MED_Hostile/Meshes/F_MED_Hostile_Dyn
bodies {
 dyn_gun [{shapeType=Capsule, length=5, radius=3, center=<1.723 0 0>, rotation=<rot 90 0 0>}]
 dyn_pouch_l [{shapeType=Capsule, length=5.2668, radius=3.963, center=<-0.5816 0 0>, rotation=<rot 90 0 0>}]
 dyn_pouch_r [{shapeType=Capsule, length=5.2668, radius=3.963, center=<-0.5816 0 0>, rotation=<rot 90 0 0>}]
 dyn_top [{shapeType=Capsule, length=5.2668, radius=3.963, center=<1.5226 0 0>, rotation=<rot 90 0 0>}]
 pack [{shapeType=Capsule, length=2, radius=2}]
}
constraints [
 {
  bone1Name=dyn_pouch_r
  bone2Name=pack
  swing1LimitDegrees=10
  swing2LimitDegrees=45
  swing2Motion=Locked
  twistLimitDegrees=5
  twistMotion=Locked
 }
 {
  bone1Name=dyn_pouch_l
  bone2Name=pack
  swing1LimitDegrees=10
  swing2LimitDegrees=45
  swing2Motion=Locked
  twistLimitDegrees=5
  twistMotion=Locked
 }
 {
  bone1Name=dyn_gun
  bone2Name=pack
  swing1LimitDegrees=10
  swing2LimitDegrees=45
  swing2Motion=Locked
  twistLimitDegrees=5
  twistMotion=Locked
 }
 {bone1Name=dyn_top, bone2Name=pack, swing1LimitDegrees=5, swing2LimitDegrees=15, twistLimitDegrees=5, twistMotion=Locked}
]
